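# Material function for the IntenseCello character that adds stylized dot rim lighting, shadow hatching and an optional fluffy robe to incoming material attributes.
# Light vector, normals, screen-aligned procedural patterns and input masks form the dot and hatch overlay, which is skipped when UseOutlines is on.
# The masked subsurface robe layers bump-offset noise for fur depth, with outline mode adding normal-push WPO and dithered opacity.
# The resulting base color is finally split between base color and emissive.
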MaterialFunction /BRCosmetics/Characters/Player/Male/Medium/Bodies/M_MED_IntenseCello/Materials/MF_IntenseCello
graph {nodes [
  n1 = ScalarParameter() {defaultValue=0.4, group=IntenseCello_Robe, parameterName=Robe_FluffWPO}
  n2 = VertexNormalWS()
  n3 = GetMaterialAttributes(None=FunctionInput() {inputName=MA_In, inputType=FunctionInput_MaterialAttributes}) {attributeGetTypes [
    #1
    #2
    #3
    #4
    #5
   ]}
  n4 = NamedRerouteUsage()
  n5 = SetMaterialAttributes!(
   MaterialAttributes=@n3.MaterialAttributes
   BaseColor=LinearInterpolate(
    A=NamedRerouteDeclaration(None=@n3.BaseColor) {name=Basecolor_IN, nodeColor=<rgba 0 1 0.1094 1>}
    B=NamedRerouteUsage()
    Alpha=NamedRerouteUsage()
   )
   Metallic=LinearInterpolate(A=@n3.Metallic, Alpha=@n4) {constB=0}
   Specular=LinearInterpolate(A=@n3.Specular, Alpha=@n4) {constB=0}
   Roughness=LinearInterpolate(A=@n3.Roughness, Alpha=@n4)
  ) {attributeSetTypes [
    #2
    #3
    #4
    #5
   ]}
  n6 = StaticSwitchParameter(True=@n3.MaterialAttributes, False=@n5) {group=IntenseCello_Robe, parameterName=UseOutlines}
  n7 = SetMaterialAttributes!(
   MaterialAttributes=@n6
   BaseColor=NamedRerouteUsage()
   Metallic=Constant()
   Specular=ScalarParameter() {defaultValue=0.5, group=IntenseCello_Robe, parameterName=Robe_Specular}
   Roughness=ScalarParameter() {defaultValue=0.9, group=IntenseCello_Robe, parameterName=Robe_Roughness}
   ShadingModel=ShadingModel() {shadingModel=MSM_Subsurface}
   SubsurfaceColor=VectorParameter() {group=IntenseCello_Robe, parameterName=Robe_SubsurfaceColor, defaultValue=<rgba 0.401 0.1211 0.2397 1>}.RGB
  ) {attributeSetTypes [
    #2
    #3
    #4
    #5
    #6
    #7
   ]}
  n8 = BlendMaterialAttributes(A=@n6, B=@n7, Alpha=NamedRerouteUsage())
  n9 = SetMaterialAttributes!(
   MaterialAttributes=@n8
   WorldPositionOffset=LinearInterpolate(B=Multiply(A=@n1, B=@n2), Alpha=NamedRerouteUsage())
   OpacityMask=NamedRerouteUsage()
  ) {attributeSetTypes [#8, #9]}
  n10 = GetMaterialAttributes(None=StaticSwitchParameter(True=StaticSwitchParameter(True=@n9, False=@n8) {group=IntenseCello_Robe, parameterName=UseOutlines}, False=@n6) {group=IntenseCello_Robe, parameterName=UseRobe, sortPriority=0}) {attributeGetTypes [#2]}
  n11 = ScalarParameter() {defaultValue=0.3, group=IntenseCello_DotRimLight, parameterName=BasecolorEmissiveRatio}
  n12 = SetMaterialAttributes!(
   MaterialAttributes=@n10.MaterialAttributes
   BaseColor=Multiply(A=@n10.BaseColor, B=OneMinus(None=@n11))
   EmissiveColor=Multiply(A=@n10.BaseColor, B=@n11)
  ) {attributeSetTypes [#2, #10]}
  FunctionOutput(None=@n12) {outputName=MA_Out}
  n13 = NamedRerouteUsage()
  LinearInterpolate(
   A=@n13
   B=MaterialFunctionCall!(Vector=@n13) {materialFunction=/Game/Packages/Fortress_SharedMaterials/Anime_Master_Material/MF_ProjectToScreenSpace}.Result
   Alpha=ScalarParameter() {group=IntenseCello_DotRimLight, parameterName=UseProjectedScreenSpaceLightVector}
  )
  MaterialFunctionCall!(Fade Offset=Constant() {r=10}, Fade Distance=Constant() {r=400}) {materialFunction=/Game/Global/Materials/Character_Parent_2023/Anime_Parent_2023/MaterialFunctions/MFC_DistanceFade_FOV_compensation}
  NamedRerouteDeclaration(None=Transform(None=NamedRerouteDeclaration(None=@n3.Normal) {name=NormalTS_In, nodeColor=<rgba 0 0.25 1 1>})) {name=NormalWS_In, nodeColor=<rgba 1 0.3516 0 1>}
  NamedRerouteDeclaration() {name=Debug, nodeColor=<rgba 0.6406 0 1 1>}
  n14 = Add(
   A=ConstantBiasScale(None=NamedRerouteUsage()) {bias=0.5, scale=2}
   B=Multiply(A=Fresnel() {baseReflectFraction=0, exponent=2}, B=Constant() {r=2})
  )
  n15 = TextureCoordinate()
  n16 = ScalarParameter() {defaultValue=7.5, group=IntenseCello_Robe, parameterName=Robe_NoiseTiling}
  n17 = TextureSample(UVs=NamedRerouteDeclaration(None=Multiply(A=@n15, B=@n16)) {name=Robe_UVs, nodeColor=<rgba 1 0.0469 0 1>}) {
   samplerType=SAMPLERTYPE_Masks
   texture=/BRCosmetics/Characters/Player/Male/Medium/Bodies/M_MED_IntenseCello/Textures/T_IntenseCello_RobeNoise
  }
  n18 = OneMinus(None=NamedRerouteDeclaration(None=@n17.R) {name=RobeNoise, nodeColor=<rgba 0 0.8125 1 1>})
  n19 = Subtract(A=Add(A=@n14) {constB=2}, B=@n18)
  n20 = ScalarParameter() {defaultValue=0.85, group=IntenseCello_Robe, parameterName=Robe_ShadingClampMin, sliderMax=1}
  n21 = Subtract(A=Add(A=@n14) {constB=1.5}, B=@n18)
  n22 = Multiply(A=Clamp(None=Saturate(None=@n19), Min=@n20) {minDefault=0.75}, B=Clamp(None=Saturate(None=@n21), Min=@n20) {minDefault=0.75})
  n23 = Add(A=@n14)
  n24 = Clamp(None=Saturate(None=Subtract(A=@n23, B=@n18)), Min=@n20) {minDefault=0.75}
  n25 = Saturate(None=Subtract(A=Add(A=@n14) {constB=0.5}, B=@n18))
  n26 = Subtract(A=Add(A=@n14) {constB=0}, B=@n18)
  n27 = Multiply(
   A=Multiply(A=Multiply(A=@n22, B=@n24), B=Clamp(None=@n25, Min=@n20) {minDefault=0.75})
   B=Clamp(None=Saturate(None=@n26), Min=@n20) {minDefault=0.75}
  )
  n28 = Add(A=@n14) {constB=-0.5}
  n29 = Clamp(None=Saturate(None=Subtract(A=@n28, B=@n18)), Min=@n20) {minDefault=0.75}
  n30 = Saturate(None=Subtract(A=Add(A=@n14) {constB=-1}, B=@n18))
  n31 = Subtract(A=Add(A=@n14) {constB=-1.5}, B=@n18)
  n32 = Multiply(
   A=Multiply(A=Multiply(A=@n27, B=@n29), B=Clamp(None=@n30, Min=@n20) {minDefault=0.75})
   B=Clamp(None=Saturate(None=@n31), Min=@n20) {minDefault=0.75}
  )
  n33 = Add(A=@n14) {constB=-2}
  n34 = Clamp(None=Saturate(None=Subtract(A=@n33, B=@n18)), Min=@n20) {minDefault=0.75}
  n35 = NamedRerouteUsage()
  NamedRerouteDeclaration(None=Multiply(A=Multiply(A=@n32, B=@n34), B=LinearInterpolate(A=@n35) {constAlpha=0.9})) {name=Robe, nodeColor=<rgba 0.1953 1 0 1>}
  NamedRerouteDeclaration(None=FunctionInput() {inputName=PreskinLocPos}) {name=PreskinLocalPos, nodeColor=<rgba 0.5469 1 0 1>}
  n36 = NamedRerouteUsage()
  n37 = NamedRerouteUsage()
  n38 = Subtract(A=@n37, B=VectorParameter() {group=IntenseCello_DotRimLight, parameterName=SphereCenterLocation, defaultValue=<rgba 0 0 130 1>}.RGB)
  n39 = Constant3Vector() {constant=<rgba 1 1 0 0>}
  n40 = Normalize(VectorInput=Transform(None=Multiply(A=@n38, B=@n39)) {transformSourceType=TRANSFORMSOURCE_Local})
  n41 = ScalarParameter() {defaultValue=0.5, group=IntenseCello_DotRimLight, parameterName=SphericalNormalBlend}
  n42 = ComponentMask(None=@n37) {b=true}
  NamedRerouteDeclaration(None=LinearInterpolate(A=@n36, B=LinearInterpolate(A=@n36, B=@n40, Alpha=@n41), Alpha=SmoothStep(Value=@n42) {constMax=35, constMin=30})) {name=SphericalNormal, nodeColor=<rgba 0.2891 1 0 1>}
  n43 = ScalarParameter() {defaultValue=0.45, group=IntenseCello_DotRimLight, parameterName=RimDotsOffset, sliderMax=0.5, sliderMin=-0.5, sortPriority=20}
  n44 = Add(A=NamedRerouteUsage() {desc=10/8: Anime shader rim light}, B=@n43) {constB=-0.25}
  n45 = ScalarParameter() {defaultValue=1.5, group=IntenseCello_DotRimLight, parameterName=RimDotsTransition, sliderMax=2, sortPriority=21}
  n46 = Subtract(
   A=Power(
    Base=Multiply(A=@n44, B=@n45) {constB=1.5}
    Exp=ScalarParameter() {defaultValue=1.5, group=IntenseCello_DotRimLight, parameterName=RimDotsPower, sortPriority=22}
   )
   B=NamedRerouteUsage()
  )
  n47 = Subtract(A=LinearInterpolate(A=@n46, B=Add(A=@n46) {constB=-0.25}, Alpha=NamedRerouteUsage()) {constA=1, constB=16}, B=NamedRerouteUsage())
  n48 = NamedRerouteDeclaration(None=SmoothStep(Min=Constant(), Max=Constant() {r=0.1}, Value=Saturate(None=@n47))) {name=RimDotsMask, nodeColor=<rgba 0.4766 1 0 1>}
  n49 = VectorParameter() {
   desc="RGB = Color\r\nA = Alpha / Instensity"
   group=IntenseCello_DotRimLight
   parameterName=RimDotsColor
   sortPriority=10
   defaultValue=<rgba 0.9 0.9 0.9 1>
  }
  n50 = Multiply(
   A=MaterialFunctionCall(
    Hue Shift Percentage=ScalarParameter() {group=IntenseCello_DotRimLight, parameterName=RimDotsColor_HueShift, sliderMax=1}
    Texture=NamedRerouteUsage()
   ) {materialFunction=&HueShift}.Result
   B=@n49.RGB
  )
  n51 = NamedRerouteUsage()
  n52 = NamedRerouteUsage()
  n53 = Saturate(None=SmoothStep(Value=Power(Base=@n52) {constExponent=3}) {constMax=0.5})
  n54 = ScalarParameter() {group=IntenseCello_DotRimLight, parameterName=RimDotsMaskIntensity}
  n55 = Multiply(
   A=LinearInterpolate(A=@n49.A, B=Multiply(A=@n49.A, B=@n51), Alpha=NamedRerouteUsage())
   B=Saturate(None=LinearInterpolate(B=@n53, Alpha=@n54) {constA=1})
  )
  n56 = ComponentMask(None=NamedRerouteUsage()) {r=true}
  NamedRerouteDeclaration(None=AppendVector(A=Multiply(A=@n48, B=@n50), B=Multiply(A=@n55, B=@n56))) {name=RimDots, nodeColor=<rgba 1 0 0.6797 1>}
  n57 = Multiply(A=NamedRerouteUsage()) {constB=-1}
  n58 = ScalarParameter() {defaultValue=0.2, group=IntenseCello_DotRimLight, parameterName=ShadowHatchingOffset, sliderMax=0.5, sliderMin=-0.5, sortPriority=20}
  n59 = Subtract(A=Multiply(A=Add(A=@n57, B=@n58) {constB=-0.25}) {constB=0.5, desc=:/}, B=NamedRerouteUsage())
  n60 = NamedRerouteDeclaration(None=SmoothStep(Min=Constant(), Max=Constant() {r=0.1}, Value=Saturate(None=@n59))) {name=ShadowHatchingMask, nodeColor=<rgba 0.3594 0.3594 0.3594 1>}
  n61 = VectorParameter() {
   desc="RGB = Color\r\nA = Alpha / Instensity"
   group=IntenseCello_DotRimLight
   parameterName=ShadowHatchingColor
   sortPriority=10
   defaultValue=<rgba 0.18 0.18 0.18 0.5>
  }
  n62 = Multiply(
   A=MaterialFunctionCall(
    Hue Shift Percentage=ScalarParameter() {group=IntenseCello_DotRimLight, parameterName=ShadowHatchingColor_HueShift, sliderMax=1}
    Texture=NamedRerouteUsage()
   ) {materialFunction=&HueShift}.Result
   B=@n61.RGB
  )
  n63 = NamedRerouteUsage()
  n64 = Power(Base=NamedRerouteUsage())
  n65 = OneMinus(None=Saturate(None=SmoothStep(Value=@n64) {constMin=0.5}))
  n66 = ScalarParameter() {group=IntenseCello_DotRimLight, parameterName=ShadowHatchingMaskIntensity}
  n67 = Multiply(
   A=LinearInterpolate(A=@n61.A, B=Multiply(A=@n61.A, B=@n63), Alpha=NamedRerouteUsage())
   B=Saturate(None=LinearInterpolate(B=@n65, Alpha=@n66) {constA=1})
  )
  n68 = ComponentMask(None=NamedRerouteUsage()) {g=true}
  NamedRerouteDeclaration(None=AppendVector(A=Multiply(A=@n60, B=@n62), B=Multiply(A=@n67, B=@n68))) {name=ShadowHatching, nodeColor=<rgba 0.099 0.099 0.099 1>}
  n69 = FunctionInput() {bUsePreviewValueAsDefault=true, inputName=M, sortPriority=1, previewValue=<0 0.5 0 1>}
  NamedRerouteDeclaration(None=ComponentMask(None=@n69) {r=true}) {name=AOMap, nodeColor=<rgba 0 1 0.7188 1>}
  n70 = NamedRerouteUsage()
  n71 = ComponentMask(None=@n70) {a=true}
  n72 = Multiply(A=@n71, B=NamedRerouteUsage())
  n73 = NamedRerouteUsage()
  n74 = Multiply(A=ComponentMask(None=@n73) {a=true}, B=NamedRerouteUsage())
  n75 = NamedRerouteUsage()
  NamedRerouteDeclaration(None=Saturate(None=LinearInterpolate(A=@n72, B=@n74, Alpha=@n75))) {name=Alpha, nodeColor=<rgba 0.6172 1 0 1>}
  n76 = MaterialFunctionCall() {materialFunction=&CameraDirectionVector}
  n77 = Multiply(A=Constant3Vector() {constant=<rgba 1 1 0 0>}, B=@n76.Result)
  n78 = Add(A=Multiply(B=Normalize(VectorInput=@n77)) {constA=0.25}, B=@n76.Result)
  n79 = Multiply(A=@n76.Result) {constB=50}
  n80 = Add(A=ActorPositionWS(), B=Constant3Vector() {constant=<rgba 0 0 90 1>})
  n81 = Constant() {r=64}
  n82 = Constant() {r=128}
  n83 = InverseLinearInterpolate(
   A=Constant() {r=45}
   B=Constant() {r=90}
   Value=MaterialFunctionCall() {materialFunction=/Engine/Functions/Engine_MaterialFunctions02/Utility/FOV}.FOV angle in degrees
  )
  NamedRerouteDeclaration(None=MaterialFunctionCall(Plane Normal Axis=Normalize(VectorInput=@n78), Plane Center=Add(A=@n79, B=@n80), UV Scale=LinearInterpolate(A=@n81, B=@n82, Alpha=@n83)) {materialFunction=&VirtualPlaneCoordinates}.UVs) {name=ObjectAlighedScreenPosition, nodeColor=<rgba 1 0.867 0 1>}
  MaterialExpressionNamedRerouteUsage_51 = NamedRerouteUsage()
  MaterialExpressionTransform_1 = Transform(None=CameraVectorWS()) {transformSourceType=TRANSFORMSOURCE_World, transformType=TRANSFORM_Tangent}
  MaterialExpressionConstant_28 = Constant() {r=4}
  MaterialExpressionConstant_27 = Constant() {r=0.05}
  MaterialExpressionTextureObject_1 = TextureObject() {
   samplerType=SAMPLERTYPE_Masks
   texture=&T_IntenseCello_RobeNoise
  }
  MaterialExpressionVertexNormalWS_1 = VertexNormalWS()
  n84 = Custom!(
   uv=@MaterialExpressionNamedRerouteUsage_51
   viewDir=@MaterialExpressionTransform_1
   steps=@MaterialExpressionConstant_28
   stepdist=@MaterialExpressionConstant_27
   tex=@MaterialExpressionTextureObject_1
   vertnormal=@MaterialExpressionVertexNormalWS_1
   dTex=@MaterialExpressionTextureObject_1
  ) {
   code="float3 rayStep = viewDir * -1;\r\nfloat2 uv2 = uv;\r\nfloat4 heightTex = Texture2DSample(tex,texSampler,uv);\r\nfloat4 distortionTex = Texture2DSample(dTex,texSampler,uv2);\r\n\r\nfor(int i=0;i<=steps;i++)\r\n{\r\nif(heightTex.r<=i/steps)\r\n{\r\nreturn float((heightTex.r*i)/steps)*1.0; \r\n}\r\nfloat stepct = heightTex.r * stepdist;\r\nuv += rayStep * stepct;\r\nheightTex = Texture2DSample(tex,texSampler,uv.xy);\r\n}\r\n\r\nreturn heightTex;\r\n"
   description=RayMarchTexture2
   outputType=CMOT_Float1
   showCode=true
   inputs [
    {
     inputName=uv
     input {expression=&self:MaterialExpressionNamedRerouteUsage_51}
    }
    {
     inputName=viewDir
     input {expression=&self:MaterialExpressionTransform_1}
    }
    {
     inputName=steps
     input {expression=&self:MaterialExpressionConstant_28}
    }
    {
     inputName=stepdist
     input {expression=&self:MaterialExpressionConstant_27}
    }
    {
     inputName=tex
     input {expression=&self:MaterialExpressionTextureObject_1}
    }
    {
     inputName=vertnormal
     input {expression=&self:MaterialExpressionVertexNormalWS_1}
    }
    {
     inputName=dTex
     input {expression=&self:MaterialExpressionTextureObject_1}
    }
   ]
  }
  NamedRerouteDeclaration(None=@n84) {name=Raymarch, nodeColor=<rgba 0.1016 0 1 1>}
  n85 = Subtract(A=Distance(A=ActorPositionWS(), B=CameraPositionWS())) {constB=100}
  NamedRerouteDeclaration(None=Saturate(None=Divide(A=@n85) {constB=400})) {name=DistanceBlending, nodeColor=<rgba 0.4766 1 0 1>}
  n86 = ScalarParameter() {defaultValue=150, group=IntenseCello_DotRimLight, parameterName=HatchingAmount, sliderMax=512, sortPriority=20}
  n87 = Multiply(A=NamedRerouteUsage(), B=LinearInterpolate(A=@n86, B=Multiply(A=@n86) {constB=0.5}, Alpha=NamedRerouteUsage()))
  n88 = Constant() {r=1}
  n89 = Frac(None=ComponentMask(None=Rotator(Coordinate=@n87, Time=@n88) {speed=0.75}) {g=true})
  NamedRerouteDeclaration(None=Abs(None=Add(A=@n89) {constB=-0.5})) {name=HatchingGradient, nodeColor=<rgba 0.3042 0 1 1>}
  NamedRerouteDeclaration(None=StaticSwitchParameter(
    True=TextureSample() {samplerType=SAMPLERTYPE_Masks, texture=/BRCosmetics/Weapons/FORT_Melee/Pickaxe_IntenseCello/Textures/T_IntenseCello_Pickaxe_FX2}.RGB
    False=Constant() {r=1}
   ) {group=IntenseCello_DotRimLight, parameterName=UsePickaxe}) {name=CustomRimHatchMask, nodeColor=<rgba 0.0547 0 1 1>}
  NamedRerouteDeclaration(None=ScalarParameter() {defaultValue=0.5, group=IntenseCello_DotRimLight, parameterName=AlphaMultiOnSkin_Hatching}) {name=AlphaMultiOnSkin_Hatching, nodeColor=<rgba 0 0.4844 1 1>}
  n90 = FunctionInput() {bUsePreviewValueAsDefault=true, inputName=TechArtMask, sortPriority=1, previewValue=<0 0.5 0 1>}
  NamedRerouteDeclaration(None=StaticSwitchParameter(True=ChannelMaskParameter(None=@n90) {group=IntenseCello_Robe, parameterName=Robe_Channel, sortPriority=1}.R, False=Constant()) {group=IntenseCello_Robe, parameterName=UseRobe, sortPriority=0}) {name=RobeMask, nodeColor=<rgba 0.7109 1 0 1>}
  n91 = ScalarParameter() {defaultValue=150, group=IntenseCello_DotRimLight, parameterName=DotsAmount, sliderMax=512, sortPriority=20}
  n92 = Multiply(A=NamedRerouteUsage(), B=LinearInterpolate(A=@n91, B=Multiply(A=@n91) {constB=0.5}, Alpha=NamedRerouteUsage()))
  n93 = Constant() {r=1}
  n94 = Add(A=Frac(None=Rotator(Coordinate=@n92, Time=@n93) {speed=0.75})) {constB=-0.5}
  n95 = Constant()
  NamedRerouteDeclaration(None=Add(A=Distance(A=@n94, B=@n95)) {constB=-0.1, desc=to get full transition}) {name=DotsGradient, nodeColor=<rgba 1 0.0938 0 1>}
  n96 = PreSkinnedNormal()
  n97 = DotProduct(A=ShaderStageSwitch(PixelShader=VertexInterpolator(VS=@n96).PS, VertexShader=@n96), B=Constant3Vector() {constant=<rgba 0 1 0 0>})
  n98 = Multiply(A=@n97, B=@n97)
  NamedRerouteDeclaration(None=LinearInterpolate(
    B=Multiply(A=@n98, B=@n98) {desc=reduce rim light on the Y axis}
    Alpha=ScalarParameter() {defaultValue=1, group=IntenseCello_DotRimLight, parameterName=RimDotsYMask_Intensity}
   )) {name=YMask, nodeColor=<rgba 0 1 0.5078 1>}
  NamedRerouteDeclaration(None=ComponentMask(None=FunctionInput() {bUsePreviewValueAsDefault=true, inputName=STT, sortPriority=1, previewValue=<0 0.5 0 1>}) {g=true}) {name=ThresholdMap, nodeColor=<rgba 0.7109 1 0 1>}
  NamedRerouteDeclaration(None=ScalarParameter() {defaultValue=0.6, group=IntenseCello_DotRimLight, parameterName=AlphaMultiOnSkin_Dots}) {name=AlphaMultiOnSkin_Dots, nodeColor=<rgba 0 0.4844 1 1>}
  NamedRerouteDeclaration(None=ComponentMask(None=@n69) {b=true}) {name=SkinMap, nodeColor=<rgba 0 1 0.1328 1>}
  n99 = LinearInterpolate(A=NamedRerouteUsage(), B=NamedRerouteUsage(), Alpha=ScalarParameter() {group=IntenseCello_DotRimLight, parameterName=UseSphericalNormal})
  n100 = Saturate(None=DotProduct(A=@n13, B=@n99))
  n101 = ScalarParameter() {defaultValue=2, group=IntenseCello_DotRimLight, parameterName=RimLight_FresnelExponent, sliderMax=12}
  n102 = Saturate(None=InverseLinearInterpolate(
    A=ScalarParameter() {group=IntenseCello_DotRimLight, parameterName=RimLight_Min, sliderMax=1}
    B=ScalarParameter() {defaultValue=1, group=IntenseCello_DotRimLight, parameterName=RimLight_Max, sliderMax=1}
    Value=Fresnel(ExponentIn=@n101, Normal=@n99) {baseReflectFraction=0}
   ) {constA=0.1, constB=0.2})
  NamedRerouteDeclaration(None=Saturate(None=Multiply(A=@n100, B=@n102))) {name=RimLight, nodeColor=<rgba 1 0 0.1172 1>}
  n103 = Multiply(A=TextureCoordinate(), B=ScalarParameter() {defaultValue=15, group=IntenseCello_Robe, parameterName=Robe_FluffNoiseTiling}) {constB=20}
  n104 = Constant()
  n105 = Constant() {r=0.1}
  n106 = TextureSample(UVs=MaterialFunctionCall!(UV Coord=@n103, Height=Constant() {r=0.25}, ReferencePlane=@n104, HeightRatio=@n105) {materialFunction=/Game/Global/MaterialFunctions/MF_BumpOffset_Accurate}.Result) {
   samplerSource=SSM_Wrap_WorldGroupSettings
   samplerType=SAMPLERTYPE_Masks
   texture=&T_IntenseCello_RobeNoise
  }
  n107 = Multiply(A=@n106.R) {constB=0.3}
  n108 = TextureSample(UVs=MaterialFunctionCall(UV Coord=@n103, Height=Constant() {r=0.5}, ReferencePlane=@n104, HeightRatio=@n105) {materialFunction=&MF_BumpOffset_Accurate}.Result) {
   samplerSource=SSM_Wrap_WorldGroupSettings
   samplerType=SAMPLERTYPE_Masks
   texture=&T_IntenseCello_RobeNoise
  }
  n109 = Multiply(A=@n108.R) {constB=0.25}
  n110 = Constant() {r=-0.2}
  n111 = Constant() {r=0.2}
  n112 = MaterialFunctionCall!(In=Add(A=@n109, B=@n110), Ramp Width=@n111) {materialFunction=/Engine/Functions/Engine_MaterialFunctions01/Math/SmoothCeil}
  n113 = TextureSample(UVs=MaterialFunctionCall(UV Coord=@n103, Height=Constant() {r=0.75}, ReferencePlane=@n104, HeightRatio=@n105) {materialFunction=&MF_BumpOffset_Accurate}.Result) {
   samplerSource=SSM_Wrap_WorldGroupSettings
   samplerType=SAMPLERTYPE_Masks
   texture=&T_IntenseCello_RobeNoise
  }
  n114 = Multiply(A=@n113.R) {constB=0.2}
  n115 = MaterialFunctionCall(In=Add(A=@n114, B=@n110), Ramp Width=@n111) {materialFunction=&SmoothCeil}
  n116 = LinearInterpolate(A=LinearInterpolate(A=@n107, B=@n109, Alpha=@n112.Result), B=@n114, Alpha=@n115.Result)
  n117 = TextureSample(UVs=MaterialFunctionCall(UV Coord=@n103, Height=Constant() {r=0.9}, ReferencePlane=@n104, HeightRatio=@n105) {materialFunction=&MF_BumpOffset_Accurate}.Result) {
   samplerSource=SSM_Wrap_WorldGroupSettings
   samplerType=SAMPLERTYPE_Masks
   texture=&T_IntenseCello_RobeNoise
  }
  n118 = Multiply(A=@n117.R) {constB=0.15}
  n119 = MaterialFunctionCall(In=Add(A=@n118, B=@n110) {constB=-0.2}, Ramp Width=@n111) {materialFunction=&SmoothCeil}
  n120 = LinearInterpolate(
   A=Multiply(A=LinearInterpolate(A=@n116, B=@n118, Alpha=@n119.Result)) {constB=4, desc=blend together layers and scale brightness}
   B=NamedRerouteUsage()
  ) {constAlpha=0.7}
  n121 = VectorParameter() {group=IntenseCello_Robe, parameterName=Robe_Color, defaultValue=<rgba 0.5677 0.2099 0.3341 1>}
  n122 = NamedRerouteUsage()
  n123 = ScalarParameter() {group=IntenseCello_Robe, parameterName=Robe_AOIntensity}
  NamedRerouteDeclaration(None=Multiply(A=Multiply(A=@n120, B=@n121.RGB), B=LinearInterpolate(B=@n122, Alpha=@n123) {constA=1})) {name=RobeColor, nodeColor=<rgba 0 1 0.1797 1>}
  n124 = ComponentMask(None=@n70) {b=true, g=true, r=true}
  n125 = ComponentMask(None=@n73) {b=true, g=true, r=true}
  NamedRerouteDeclaration(None=LinearInterpolate(A=Multiply(A=@n124, B=@n71) {desc=Rim Dots}, B=Multiply(A=@n125, B=@n74) {desc=Shadow Hatching}, Alpha=@n75) {desc=Avoid dot draw on top of shadows}) {name=Color, nodeColor=<rgba 1 0.0469 0 1>}
  n126 = Custom!() {code="return View.DirectionalLightDirection;", description=DirectionalLightVector, inputs []}
  n127 = DotProduct(A=@n126, B=@n126)
  n128 = CollectionParameter() {
   collection=&FortniteMaterialParameters
   parameterName=SunlightDirectionalVector
  }
  n129 = ComponentMask(None=@n128) {b=true, g=true, r=true}
  n130 = LinearInterpolate(
   A=If(
    A=Custom() {code="return View.DirectionalLightColor;", description=DirectionalLightColor, inputs []}
    "A > B"=If(A=@n127, "A > B"=@n126, "A == B"=@n129, "A < B"=@n129) {desc=check if DirectionalLight exists}
    "A == B"=@n129
    "A < B"=@n129
   ) {desc=check if DirectionalLight color is Black}
   B=Transform(None=VectorParameter() {group=IntenseCello_DotRimLight, parameterName=LightVector, defaultValue=<rgba -1 0 0 1>}.RGB) {transformSourceType=TRANSFORMSOURCE_View}
   Alpha=ScalarParameter() {group=IntenseCello_DotRimLight, parameterName=UseFixedLightVector}
  )
  NamedRerouteDeclaration(None=DotProduct(A=NamedRerouteDeclaration(None=@n130) {name=LightVector, nodeColor=<rgba 1 0.4688 0 1>}, B=NamedRerouteUsage())) {name=LdotN, nodeColor=<rgba 0.2656 0 1 1>}
  n131 = MaterialFunctionCall(In=Add(A=@n107, B=@n110), Ramp Width=@n111) {materialFunction=&SmoothCeil}
  n132 = Add(A=Add(A=Add(A=@n131.Result, B=@n112.Result), B=@n115.Result), B=@n119.Result) {desc=accumulate alpah masks}
  n133 = Power(Base=OneMinus(None=Fresnel() {baseReflectFraction=0, exponent=1})) {constExponent=4, desc="make core more visable to hide AO "}
  n134 = Saturate(None=Subtract(A=Add(A=@n132, B=@n133), B=Fresnel() {baseReflectFraction=0, desc=to make the edge more fuzzy, exponent=4})) {desc=opacity}
  NamedRerouteDeclaration(None=Multiply(
    A=NamedRerouteUsage()
    B=MaterialFunctionCall!(Alpha Threshold=@n134, Random) {materialFunction=/Game/Environments/Asteria/Sets/NeonCity/Techart/Screen/MaterialFunctions/DitherTemporalAABlue}.Result
   )) {name=RobeOpacity, nodeColor=<rgba 0.125 1 0 1>}
  PixelNormalWS()
  StaticSwitchParameter(
   True=SetMaterialAttributes!(MaterialAttributes, BaseColor=NamedRerouteUsage()) {attributeSetTypes [#2]}
   False=@n12
  ) {group=IntenseCello_DotRimLight, parameterName=DebugRim}
 ]}
