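# Parent material for the IntenseCello medium male player character body, using material attributes and expression-driven shading models, with skeletal mesh, cloth and Nanite usage.
# The shared 2023 character PreFX function supplies attributes and masks, and static switches optionally blend in FN hair and apply flyaway hair.
# The result runs through the character-specific MF_IntenseCello, fed pre-skinned local position and masks, then the shared character PostFX and cinematic PostFX functions.
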Material /BRCosmetics/Characters/Player/Male/Medium/Bodies/M_MED_IntenseCello/Materials/M_IntenseCello_Parent
graph {
 properties {
  bAutomaticallySetUsageInEditor=false
  bHasPixelAnimation=true
  bUseMaterialAttributes=true
  bUsedWithClothing=true
  bUsedWithMeshDeformer=true
  bUsedWithMorphTargets=true
  bUsedWithNanite=true
  bUsedWithSkeletalMesh=true
  pixelDepthOffsetMode=PDOM_Legacy
  shadingModel=MSM_FromMaterialExpression
  subsurfaceProfile=&SS_HeroSkin_02
  usedShadingModels=Default Lit | Subsurface | Clear Coat | Subsurface Profile | Cloth
 }
 nodes [
  n1 = MaterialFunctionCall() {materialFunction=&MF_CharacterParent_2023_PreFX}
  n2 = MaterialFunctionCall!(In=@n1.Result, Mask Input=@n1.M) {materialFunction=/Game/Global/Materials/Character_Parent_2023/Hair_Parent_2023/MaterialFunctions/MF_FN_Hair_2023}
  n3 = BlendMaterialAttributes(A=@n1.Result, B=@n2.Result, Alpha=@n2.Result)
  n4 = StaticSwitchParameter(
   True=StaticSwitchParameter(
    True=MaterialFunctionCall!(MaterialInput=@n3) {materialFunction=/Game/Global/Materials/Character_Parent_2023/Hair_Parent_2023/MaterialFunctions/MF_FlyAwayHair_2023}.Result
    False=@n3
   ) {group=Skin FX, parameterName=IsFlyAwayHair?}
   False=@n1.Result
  ) {group=Skin FX, parameterName=UseFNHair2_0}
  n5 = NamedRerouteUsage()
  n6 = NamedRerouteUsage()
  n7 = NamedRerouteUsage()
  MaterialFunctionCall(In=MaterialFunctionCall(
    Material=MaterialFunctionCall!(MA_In=@n4, PreskinLocPos=@n5, M=@n6, TechArtMask=@n7, STT) {materialFunction=&/MF_IntenseCello}.MA_Out
    M=NamedRerouteDeclaration(None=@n1.M) {name=M, nodeColor=<rgba 0 0.1094 1 1>}
    Pre Skinned Local Position=@n1.Pre skinned Local Position
    Distance Blend=@n1.Pre skinned Local Position
    Fresnel=@n1.Pre skinned Local Position
   ) {materialFunction=&MF_CharacterParent_2023_PostFX}.Result) {materialFunction=&CINE_MF_CharPostFX}
  NamedRerouteDeclaration(None=@n1.Tech Art Mask) {name=TechArtMask, nodeColor=<rgba 0 1 0.2969 1>}
  NamedRerouteDeclaration(None=@n1.Pre skinned Local Position) {name=PreskinLocPos, nodeColor=<rgba 0 1 0.2031 1>}
 ]
}
