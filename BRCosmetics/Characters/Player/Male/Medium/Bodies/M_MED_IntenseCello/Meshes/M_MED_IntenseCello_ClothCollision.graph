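# Physics asset for cloth collision on the M_MED_IntenseCello male medium body cosmetic.
# Capsules cover the pelvis, two lower spine bones and the left and right dynamic belt chains, with one near-identical capsule reused along the belts.
# A large flat box on dyn_ground acts as a floor.
# Leg, foot and forearm bodies hold no shapes, and the three left-leg constraints leave swing and twist free.
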PhysicsAsset /BRCosmetics/Characters/Player/Male/Medium/Bodies/M_MED_IntenseCello/Meshes/M_MED_IntenseCello_ClothCollision
bodies {
 calf_l []
 calf_r []
 dyn_belt_l_01 [{shapeType=Capsule, length=11.8293, radius=8, shapeName=dyn_belt_r_01_capsule, center=<3.5545 0 6.6896>, rotation=<rot 90 0 0>}]
 dyn_belt_l_02 [{shapeType=Capsule, length=11.8293, radius=8, shapeName=dyn_belt_r_01_capsule, center=<3.5545 0 6.6896>, rotation=<rot 90 0 0>}]
 dyn_belt_l_03 [{shapeType=Capsule, length=11.8293, radius=8, shapeName=dyn_belt_r_01_capsule, center=<3.5545 0 6.6896>, rotation=<rot 90 0 0>}]
 dyn_belt_l_04 [{shapeType=Capsule, length=11.8293, radius=8, shapeName=dyn_belt_r_01_capsule, center=<3.5545 0 6.6896>, rotation=<rot 90 0 0>}]
 dyn_belt_r_01 [{shapeType=Capsule, length=11.8293, radius=8, shapeName=dyn_belt_r_01_capsule, center=<3.5545 0 6.6896>, rotation=<rot 90 0 0>}]
 dyn_belt_r_02 [{shapeType=Capsule, length=11.8293, radius=8, shapeName=dyn_belt_r_01_capsule, center=<3.5545 0 6.6896>, rotation=<rot 90 0 0>}]
 dyn_belt_r_03 [{shapeType=Capsule, length=11.8293, radius=8, shapeName=dyn_belt_r_01_capsule, center=<3.5545 0 6.3984>, rotation=<rot 90 0 0>}]
 dyn_belt_r_04 [{shapeType=Capsule, length=11.8293, radius=8, shapeName=dyn_belt_r_01_capsule, center=<3.5545 0 6.3984>, rotation=<rot 90 0 0>}]
 dyn_ground [{shapeType=Box, extentX=327.5, extentY=327.5, extentZ=59.22, shapeName=root_box, center=<0 0 -29.0385>}]
 foot_l []
 foot_r []
 lowerarm_l []
 lowerarm_r []
 pelvis [{shapeType=Capsule, length=8.5164, radius=10.7208, center=<-0.0668 0.4826 0>}]
 spine_01 [{shapeType=Capsule, length=7.1527, radius=9.7082, center=<0.3095 -0.6696 0>}]
 spine_02 [{shapeType=Capsule, length=8.5692, radius=8.9596, center=<-1.0816 -0.436 -0.0002>}]
 thigh_l []
 thigh_r []
}
constraints [
 {
  bone1Name=calf_l
  bone2Name=thigh_l
  swing1LimitDegrees=45
  swing1Motion=Free
  swing2LimitDegrees=45
  swing2Motion=Free
  twistLimitDegrees=45
  twistMotion=Free
 }
 {
  bone1Name=foot_l
  bone2Name=thigh_l
  swing1LimitDegrees=45
  swing1Motion=Free
  swing2LimitDegrees=45
  swing2Motion=Free
  twistLimitDegrees=45
  twistMotion=Free
 }
 {
  bone1Name=thigh_l
  bone2Name=pelvis
  swing1LimitDegrees=45
  swing1Motion=Free
  swing2LimitDegrees=45
  swing2Motion=Free
  twistLimitDegrees=45
  twistMotion=Free
 }
]
