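# Physics asset for the dynamic belt of the IntenseCello medium male player body, named as a variant for cloth.
# Left and right belt bones form four-link capsule chains off tiny master bodies, joined by constraints with identical swing and tight twist limits.
# Unconstrained capsules on the pelvis, lower spine, thighs, calves (two each) and back coat bones supply surrounding collision volumes.
# Many shape names are reused from other bones, such as coat names on belt capsules.
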PhysicsAsset /BRCosmetics/Characters/Player/Male/Medium/Bodies/M_MED_IntenseCello/Meshes/M_MED_IntenseCello_Dyn_belt_for_cloth
bodies {
 calf_l [
  {
   shapeType=Capsule
   length=20.6772
   radius=12.4225
   shapeName=calf_l_capsule
   center=<-44.1542 0.83 -1.0265>
   rotation=<rot 82.1202 178.1258 -170.9083>
  }
  {
   shapeType=Capsule
   length=29.2275
   radius=10.2928
   shapeName=calf_l_capsule
   center=<-20.0091 -1.4557 0.3034>
   rotation=<rot 85.3254 -175.5151 178.9213>
  }
 ]
 calf_r [
  {
   shapeType=Capsule
   length=20.6772
   radius=12.4225
   shapeName=calf_l_capsule
   center=<44.1542 -0.83 1.0265>
   rotation=<rot 82.1202 178.1258 9.0917>
  }
  {
   shapeType=Capsule
   length=29.2275
   radius=10.2928
   shapeName=calf_l_capsule
   center=<20.0091 1.4557 -0.3034>
   rotation=<rot 85.3254 -175.5151 -1.0787>
  }
 ]
 dyn_belt_l_01 [{shapeType=Capsule, length=4.4902, radius=5, shapeName=dyn_coat_l_01_capsule, center=<9.1939 -1.2679 5>, rotation=<rot 90 0 0>}]
 dyn_belt_l_02 [{shapeType=Capsule, length=7.6215, radius=5, shapeName=dyn_coat_l_01_capsule, center=<4.4255 0 5>, rotation=<rot 90 0 0>}]
 dyn_belt_l_03 [{shapeType=Capsule, length=10.298, radius=7.0018, shapeName=dyn_coat_l_01_capsule, center=<6.5159 0 5.8785>, rotation=<rot 90 0 0>}]
 dyn_belt_l_04 [{shapeType=Capsule, length=9.3303, radius=6.8018, shapeName=dyn_coat_l_01_capsule, center=<6.3224 0 4.6326>, rotation=<rot 90 0 0>}]
 dyn_belt_l_MSTR [{shapeType=Capsule, length=0.505, radius=0.505, shapeName=dyn_coat_l_MSTR_capsule}]
 dyn_belt_r_01 [{shapeType=Capsule, length=4.4902, radius=5, shapeName=dyn_coat_l_01_capsule, center=<9.9602 1.3985 5.3041>, rotation=<rot 90 0 0>}]
 dyn_belt_r_02 [{shapeType=Capsule, length=7.6215, radius=5, shapeName=dyn_coat_l_01_capsule, center=<4.8977 0.0009 5.3317>, rotation=<rot 90 0 0>}]
 dyn_belt_r_03 [{shapeType=Capsule, length=9.8231, radius=7.0018, shapeName=dyn_coat_l_01_capsule, center=<7.0901 0.0009 6.845>, rotation=<rot 90 0 0>}]
 dyn_belt_r_04 [{shapeType=Capsule, length=10.8177, radius=6.8018, shapeName=dyn_coat_l_01_capsule, center=<7.1818 0.0011 6.5436>, rotation=<rot 90 0 0>}]
 dyn_belt_r_MSTR [{shapeType=Capsule, length=0.505, radius=0.505, shapeName=dyn_coat_l_MSTR_capsule}]
 dyn_coat_bk_l_02 [{
   shapeType=Capsule
   length=23.217
   radius=13.2579
   shapeName=dyn_coat_l_01_capsule
   center=<13.1114 0.0003 -13.5214>
   rotation=<rot 83.7388 -179.9998 -179.9998>
  }]
 dyn_coat_bk_l_04 [{
   shapeType=Capsule
   length=21.4567
   radius=15.0182
   shapeName=dyn_coat_l_01_capsule
   center=<9.8007 0.825 -14.3638>
   rotation=<rot 90 179.5524 174.2434>
  }]
 dyn_coat_bk_l_05 [{
   shapeType=Capsule
   length=19.4749
   radius=17
   shapeName=dyn_coat_l_01_capsule
   center=<8.5281 0.0005 -16.2437>
   rotation=<rot 87.02 0.0004 0.0004>
  }]
 dyn_coat_bk_r_02 [{shapeType=Capsule, length=21.967, radius=14.5079, shapeName=dyn_coat_l_01_capsule, center=<13.19 1.6422 -13.9626>, rotation=<rot 90 0 0>}]
 dyn_coat_bk_r_04 [{shapeType=Capsule, length=20.2067, radius=16.2682, shapeName=dyn_coat_l_01_capsule, center=<9.8349 1.322 -15.1112>, rotation=<rot 90 0 0>}]
 dyn_coat_bk_r_05 [{shapeType=Capsule, length=18.2249, radius=18.25, shapeName=dyn_coat_l_01_capsule, center=<8.5397 0.0003 -17>, rotation=<rot 90 0 0>}]
 pelvis [{shapeType=Capsule, length=12.0617, radius=10.4987, shapeName=pelvis_capsule, center=<-3.8176 -2.1282 0>}]
 spine_01 [{shapeType=Capsule, length=6.84, radius=10.4987, shapeName=pelvis_capsule, center=<-0.9991 -1.1308 0>}]
 thigh_l [{shapeType=Capsule, length=30.8932, radius=11.7277, shapeName=pelvis_capsule, center=<-19.8854 2.3454 -0.0134>, rotation=<rot 90 0 0>}]
 thigh_r [{shapeType=Capsule, length=30.8932, radius=11.7277, shapeName=pelvis_capsule, center=<19.9619 -0.8746 -0.2005>, rotation=<rot 90 0 0>}]
}
constraints [
 {bone1Name=dyn_belt_r_01, bone2Name=dyn_belt_r_MSTR, swing1LimitDegrees=20, swing2LimitDegrees=45, twistLimitDegrees=5}
 {bone1Name=dyn_belt_r_02, bone2Name=dyn_belt_r_01, swing1LimitDegrees=20, swing2LimitDegrees=45, twistLimitDegrees=5}
 {bone1Name=dyn_belt_r_03, bone2Name=dyn_belt_r_02, swing1LimitDegrees=20, swing2LimitDegrees=45, twistLimitDegrees=5}
 {bone1Name=dyn_belt_r_04, bone2Name=dyn_belt_r_03, swing1LimitDegrees=20, swing2LimitDegrees=45, twistLimitDegrees=5}
 {bone1Name=dyn_belt_l_01, bone2Name=dyn_belt_l_MSTR, swing1LimitDegrees=20, swing2LimitDegrees=45, twistLimitDegrees=5}
 {bone1Name=dyn_belt_l_02, bone2Name=dyn_belt_l_01, swing1LimitDegrees=20, swing2LimitDegrees=45, twistLimitDegrees=5}
 {bone1Name=dyn_belt_l_03, bone2Name=dyn_belt_l_02, swing1LimitDegrees=20, swing2LimitDegrees=45, twistLimitDegrees=5}
 {bone1Name=dyn_belt_l_04, bone2Name=dyn_belt_l_03, swing1LimitDegrees=20, swing2LimitDegrees=45, twistLimitDegrees=5}
]
