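# Physics asset for the dynamic coat of the IntenseCello medium male player body.
# Five bone chains (front left/right, back center/left/right) hang from spine-linked master bones with limited swing and tight twist, using capsules and large boxes.
# Angle-free cross constraints tie front chains to back sides, and back sides to the center.
# Spine, pelvis and leg capsules plus a ground box provide body and floor shapes; spine_04 constraints lack a matching body.
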PhysicsAsset /BRCosmetics/Characters/Player/Male/Medium/Bodies/M_MED_IntenseCello/Meshes/M_MED_IntenseCello_Dyn_coat
bodies {
 calf_l [{shapeType=Capsule, length=50.6561, radius=13.1273, shapeName=pelvis_capsule, center=<-30.4611 2.7534 4.4558>, rotation=<rot 90 0 0>}]
 calf_r [{shapeType=Capsule, length=50.6561, radius=13.1273, shapeName=pelvis_capsule, center=<30.4611 -2.7534 -4.4558>, rotation=<rot 90 180 0>}]
 dyn_coat_bk_01 [{shapeType=Capsule, length=19.2053, radius=20, shapeName=dyn_coat_l_01_capsule, center=<8.7723 0.0007 20>, rotation=<rot 90 0 0>}]
 dyn_coat_bk_02 [{shapeType=Box, extentX=40, extentY=38, extentZ=18, shapeName=dyn_coat_bk_r_05_box, center=<9.5968 0.0006 8.5323>}]
 dyn_coat_bk_04 [{shapeType=Box, extentX=40, extentY=40, extentZ=18, shapeName=dyn_coat_bk_r_05_box, center=<9.5968 0.0006 8.5323>}]
 dyn_coat_bk_05 [{shapeType=Box, extentX=40, extentY=40, extentZ=18, shapeName=dyn_coat_bk_r_05_box, center=<3.4495 0.0006 8.5323>}]
 dyn_coat_bk_06 [{shapeType=Box, extentX=40, extentY=40, extentZ=18, shapeName=dyn_coat_bk_r_05_box, center=<-0.3165 0.0006 8.5323>}]
 dyn_coat_bk_MSTR [{shapeType=Capsule, length=0.505, radius=0.505, shapeName=dyn_coat_l_MSTR_capsule}]
 dyn_coat_bk_l_01 [{shapeType=Capsule, length=25.8014, radius=20, shapeName=dyn_coat_l_01_capsule, center=<10.7079 0 20>, rotation=<rot 90 0 0>}]
 dyn_coat_bk_l_02 [{shapeType=Capsule, length=21.2561, radius=20, shapeName=dyn_coat_l_01_capsule, center=<10.3356 0 20>, rotation=<rot 90 0 0>}]
 dyn_coat_bk_l_04 [{shapeType=Capsule, length=21.2561, radius=20, shapeName=dyn_coat_l_01_capsule, center=<10.3356 0 20>, rotation=<rot 90 0 0>}]
 dyn_coat_bk_l_05 [{shapeType=Box, extentX=40, extentY=37, extentZ=18, shapeName=dyn_coat_bk_r_05_box, center=<3.4495 2.9723 6.2633>, rotation=<rot 0 0 3.7679>}]
 dyn_coat_bk_l_06 [{shapeType=Box, extentX=40, extentY=37, extentZ=18, shapeName=dyn_coat_bk_r_05_box, center=<-0.9952 2.9723 6.2633>, rotation=<rot 0 0 3.7679>}]
 dyn_coat_bk_l_MSTR [{shapeType=Capsule, length=0.505, radius=0.505, shapeName=dyn_coat_l_MSTR_capsule}]
 dyn_coat_bk_r_01 [{shapeType=Capsule, length=25.8014, radius=20, shapeName=dyn_coat_l_01_capsule, center=<10.7079 0 20>, rotation=<rot 90 0 0>}]
 dyn_coat_bk_r_02 [{shapeType=Capsule, length=21.2561, radius=20, shapeName=dyn_coat_l_01_capsule, center=<10.3356 0 20>, rotation=<rot 90 0 0>}]
 dyn_coat_bk_r_04 [{shapeType=Capsule, length=21.2561, radius=20, shapeName=dyn_coat_l_01_capsule, center=<10.3356 0 20>, rotation=<rot 90 0 0>}]
 dyn_coat_bk_r_05 [{shapeType=Box, extentX=40, extentY=37, extentZ=18, shapeName=dyn_coat_bk_r_05_box, center=<3.4495 -4.4091 5.4157>, rotation=<rot 0 0 0.338>}]
 dyn_coat_bk_r_06 [{shapeType=Box, extentX=40, extentY=37, extentZ=18, shapeName=dyn_coat_bk_r_05_box, center=<-0.9952 -4.4091 5.4157>, rotation=<rot 0 0 0.338>}]
 dyn_coat_bk_r_MSTR [{shapeType=Capsule, length=0.505, radius=0.505, shapeName=dyn_coat_l_MSTR_capsule}]
 dyn_coat_l_01 [{shapeType=Capsule, length=10.7561, radius=8, shapeName=dyn_coat_l_01_capsule, center=<4.667 0 6.211>, rotation=<rot 90 0 0>}]
 dyn_coat_l_02 [{shapeType=Capsule, length=12.2561, radius=8, shapeName=dyn_coat_l_01_capsule, center=<5.44 0 5.5818>, rotation=<rot 90 0 0>}]
 dyn_coat_l_03 [{shapeType=Capsule, length=21.2561, radius=8, shapeName=dyn_coat_l_01_capsule, center=<10.3356 0 5.0234>, rotation=<rot 90 0 0>}]
 dyn_coat_l_04 [{shapeType=Capsule, length=21.2561, radius=15, shapeName=dyn_coat_l_01_capsule, center=<10.3356 0 15.4045>, rotation=<rot 90 0 0>}]
 dyn_coat_l_05 [{shapeType=Capsule, length=11.0061, radius=15, shapeName=dyn_coat_l_01_capsule, center=<3.5998 0 16.4623>, rotation=<rot 90 0 0>}]
 dyn_coat_l_06 [{shapeType=Box, extentX=40, extentY=37, extentZ=18, shapeName=dyn_coat_bk_r_05_box, center=<4.0962 2.9723 6.2633>, rotation=<rot 0 0 3.7679>}]
 dyn_coat_l_MSTR [{shapeType=Capsule, length=0.505, radius=0.505, shapeName=dyn_coat_l_MSTR_capsule}]
 dyn_coat_r_01 [{shapeType=Capsule, length=10.7561, radius=8, shapeName=dyn_coat_l_01_capsule, center=<4.667 0 6.211>, rotation=<rot 90 0 0>}]
 dyn_coat_r_02 [{shapeType=Capsule, length=12.2561, radius=8, shapeName=dyn_coat_l_01_capsule, center=<5.44 0 5.5818>, rotation=<rot 90 0 0>}]
 dyn_coat_r_03 [{shapeType=Capsule, length=21.2561, radius=8, shapeName=dyn_coat_l_01_capsule, center=<10.3356 0 5.0234>, rotation=<rot 90 0 0>}]
 dyn_coat_r_04 [{shapeType=Capsule, length=21.2561, radius=15, shapeName=dyn_coat_l_01_capsule, center=<10.3356 0 15.4045>, rotation=<rot 90 0 0>}]
 dyn_coat_r_05 [{shapeType=Capsule, length=8.7263, radius=17.2798, shapeName=dyn_coat_l_01_capsule, center=<3.5998 0 16.4623>, rotation=<rot 90 0 0>}]
 dyn_coat_r_06 [{shapeType=Box, extentX=40, extentY=37, extentZ=18, shapeName=dyn_coat_bk_r_05_box, center=<3.5038 -4.4091 5.4157>, rotation=<rot 0 0 0.338>}]
 dyn_coat_r_MSTR [{shapeType=Capsule, length=0.505, radius=0.505, shapeName=dyn_coat_l_MSTR_capsule}]
 dyn_ground [{shapeType=Box, extentX=327.5, extentY=327.5, extentZ=59.22, shapeName=root_box, center=<0 0 -24.8673>}]
 pelvis [{shapeType=Capsule, length=6.967, radius=11.521, shapeName=pelvis_capsule, center=<-4.8171 -0.6518 -0.0005>}]
 spine_01 [{shapeType=Capsule, length=6.4652, radius=10.9987, shapeName=pelvis_capsule, center=<-0.9991 -0.3721 0>}]
 spine_02 [{shapeType=Capsule, length=7.4401, radius=10.4987, shapeName=pelvis_capsule, center=<0.1451 -0.7547 0>}]
 spine_03 [{shapeType=Capsule, length=7.4401, radius=10.4987, shapeName=pelvis_capsule, center=<2.1769 0.258 0>}]
 thigh_l [{shapeType=Capsule, length=30.089, radius=10.5541, shapeName=pelvis_capsule, center=<-20.0337 -2.1741 2.9929>, rotation=<rot 86.3502 180 180>}]
 thigh_r [{shapeType=Capsule, length=30.089, radius=10.5541, shapeName=pelvis_capsule, center=<20.0337 2.1741 -2.9929>, rotation=<rot 86.3502 180 0>}]
}
constraints [
 {bone1Name=dyn_coat_l_01, bone2Name=dyn_coat_l_MSTR, swing1LimitDegrees=45, swing2LimitDegrees=45, twistLimitDegrees=5}
 {bone1Name=dyn_coat_l_02, bone2Name=dyn_coat_l_01, swing1LimitDegrees=25, swing2LimitDegrees=45, twistLimitDegrees=5}
 {bone1Name=dyn_coat_l_03, bone2Name=dyn_coat_l_02, swing1LimitDegrees=10, swing2LimitDegrees=45, twistLimitDegrees=5}
 {bone1Name=dyn_coat_l_04, bone2Name=dyn_coat_l_03, swing1LimitDegrees=10, swing2LimitDegrees=45, twistLimitDegrees=5}
 {bone1Name=dyn_coat_l_05, bone2Name=dyn_coat_l_04, swing1LimitDegrees=10, swing2LimitDegrees=45, twistLimitDegrees=5}
 {bone1Name=dyn_coat_l_06, bone2Name=dyn_coat_l_05, swing1LimitDegrees=10, swing2LimitDegrees=45, twistLimitDegrees=5}
 {bone1Name=dyn_coat_r_01, bone2Name=dyn_coat_r_MSTR, swing1LimitDegrees=45, swing2LimitDegrees=45, twistLimitDegrees=5}
 {bone1Name=dyn_coat_r_02, bone2Name=dyn_coat_r_01, swing1LimitDegrees=25, swing2LimitDegrees=45, twistLimitDegrees=5}
 {bone1Name=dyn_coat_r_03, bone2Name=dyn_coat_r_02, swing1LimitDegrees=10, swing2LimitDegrees=45, twistLimitDegrees=5}
 {bone1Name=dyn_coat_r_04, bone2Name=dyn_coat_r_03, swing1LimitDegrees=10, swing2LimitDegrees=45, twistLimitDegrees=5}
 {bone1Name=dyn_coat_r_05, bone2Name=dyn_coat_r_04, swing1LimitDegrees=10, swing2LimitDegrees=45, twistLimitDegrees=5}
 {bone1Name=dyn_coat_r_06, bone2Name=dyn_coat_r_05, swing1LimitDegrees=10, swing2LimitDegrees=45, twistLimitDegrees=5}
 {bone1Name=dyn_coat_bk_01, bone2Name=dyn_coat_bk_MSTR, swing1LimitDegrees=10, swing2LimitDegrees=15, twistLimitDegrees=5}
 {bone1Name=dyn_coat_bk_02, bone2Name=dyn_coat_bk_01, swing1LimitDegrees=10, swing2LimitDegrees=15, twistLimitDegrees=5}
 {bone1Name=dyn_coat_bk_l_01, bone2Name=dyn_coat_bk_l_MSTR, swing1LimitDegrees=10, swing2LimitDegrees=45, twistLimitDegrees=5}
 {bone1Name=dyn_coat_bk_l_02, bone2Name=dyn_coat_bk_l_01, swing1LimitDegrees=15, swing2LimitDegrees=45, twistLimitDegrees=5}
 {bone1Name=dyn_coat_bk_r_01, bone2Name=dyn_coat_bk_r_MSTR, swing1LimitDegrees=10, swing2LimitDegrees=45, twistLimitDegrees=5}
 {bone1Name=dyn_coat_bk_r_02, bone2Name=dyn_coat_bk_r_01, swing1LimitDegrees=15, swing2LimitDegrees=45, twistLimitDegrees=5}
 {
  bone1Name=calf_l
  bone2Name=thigh_l
  swing1LimitDegrees=45
  swing1Motion=Free
  swing2LimitDegrees=45
  swing2Motion=Free
  twistLimitDegrees=45
  twistMotion=Free
 }
 {
  bone1Name=dyn_coat_l_MSTR
  bone2Name=spine_03
  swing1LimitDegrees=45
  swing1Motion=Free
  swing2LimitDegrees=45
  swing2Motion=Free
  twistLimitDegrees=45
  twistMotion=Free
 }
 {
  bone1Name=dyn_coat_r_MSTR
  bone2Name=spine_03
  swing1LimitDegrees=45
  swing1Motion=Free
  swing2LimitDegrees=45
  swing2Motion=Free
  twistLimitDegrees=45
  twistMotion=Free
 }
 {
  bone1Name=dyn_coat_bk_MSTR
  bone2Name=spine_01
  swing1LimitDegrees=45
  swing1Motion=Free
  swing2LimitDegrees=45
  swing2Motion=Free
  twistLimitDegrees=45
  twistMotion=Free
 }
 {
  bone1Name=dyn_coat_bk_l_MSTR
  bone2Name=spine_01
  swing1LimitDegrees=45
  swing1Motion=Free
  swing2LimitDegrees=45
  swing2Motion=Free
  twistLimitDegrees=45
  twistMotion=Free
 }
 {
  bone1Name=dyn_coat_bk_r_MSTR
  bone2Name=spine_01
  swing1LimitDegrees=45
  swing1Motion=Free
  swing2LimitDegrees=45
  swing2Motion=Free
  twistLimitDegrees=45
  twistMotion=Free
 }
 {
  bone1Name=thigh_l
  bone2Name=pelvis
  swing1LimitDegrees=45
  swing1Motion=Free
  swing2LimitDegrees=45
  swing2Motion=Free
  twistLimitDegrees=45
  twistMotion=Free
 }
 {
  bone1Name=spine_01
  bone2Name=pelvis
  swing1LimitDegrees=45
  swing1Motion=Free
  swing2LimitDegrees=45
  swing2Motion=Free
  twistLimitDegrees=45
  twistMotion=Free
 }
 {
  bone1Name=spine_02
  bone2Name=spine_01
  swing1LimitDegrees=45
  swing1Motion=Free
  swing2LimitDegrees=45
  swing2Motion=Free
  twistLimitDegrees=45
  twistMotion=Free
 }
 {
  bone1Name=spine_03
  bone2Name=spine_02
  swing1LimitDegrees=45
  swing1Motion=Free
  swing2LimitDegrees=45
  swing2Motion=Free
  twistLimitDegrees=45
  twistMotion=Free
 }
 {bone1Name=dyn_coat_bk_r_05, bone2Name=dyn_coat_bk_r_04, swing1LimitDegrees=20, swing2LimitDegrees=45, twistLimitDegrees=5}
 {bone1Name=dyn_coat_bk_r_06, bone2Name=dyn_coat_bk_r_05, swing1LimitDegrees=10, swing2LimitDegrees=20, twistLimitDegrees=5}
 {bone1Name=dyn_coat_bk_l_05, bone2Name=dyn_coat_bk_l_04, swing1LimitDegrees=20, swing2LimitDegrees=45, twistLimitDegrees=5}
 {bone1Name=dyn_coat_bk_l_06, bone2Name=dyn_coat_bk_l_05, swing1LimitDegrees=10, swing2LimitDegrees=20, twistLimitDegrees=5}
 {bone1Name=dyn_coat_bk_r_04, bone2Name=dyn_coat_bk_r_02, swing1LimitDegrees=20, swing2LimitDegrees=45, twistLimitDegrees=5}
 {bone1Name=dyn_coat_bk_04, bone2Name=dyn_coat_bk_02, swing1LimitDegrees=10, swing2LimitDegrees=15, twistLimitDegrees=5}
 {bone1Name=dyn_coat_bk_l_04, bone2Name=dyn_coat_bk_l_02, swing1LimitDegrees=20, swing2LimitDegrees=45, twistLimitDegrees=5}
 {
  bone1Name=dyn_coat_r_05
  bone2Name=dyn_coat_bk_r_05
  swing1LimitDegrees=45
  swing1Motion=Free
  swing2LimitDegrees=45
  swing2Motion=Free
  twistLimitDegrees=45
  twistMotion=Free
 }
 {
  bone1Name=dyn_coat_l_05
  bone2Name=dyn_coat_bk_l_05
  swing1LimitDegrees=45
  swing1Motion=Free
  swing2LimitDegrees=45
  swing2Motion=Free
  twistLimitDegrees=45
  twistMotion=Free
 }
 {
  bone1Name=dyn_coat_r_04
  bone2Name=dyn_coat_bk_r_02
  swing1LimitDegrees=45
  swing1Motion=Free
  swing2LimitDegrees=45
  swing2Motion=Free
  twistLimitDegrees=45
  twistMotion=Free
 }
 {
  bone1Name=dyn_coat_l_04
  bone2Name=dyn_coat_bk_l_02
  swing1LimitDegrees=45
  swing1Motion=Free
  swing2LimitDegrees=45
  swing2Motion=Free
  twistLimitDegrees=45
  twistMotion=Free
 }
 {bone1Name=dyn_coat_bk_05, bone2Name=dyn_coat_bk_04, swing1LimitDegrees=10, swing2LimitDegrees=15, twistLimitDegrees=5}
 {bone1Name=dyn_coat_bk_06, bone2Name=dyn_coat_bk_05, swing1LimitDegrees=10, swing2LimitDegrees=15, twistLimitDegrees=5}
 {
  bone1Name=calf_r
  bone2Name=thigh_r
  swing1LimitDegrees=45
  swing1Motion=Free
  swing2LimitDegrees=45
  swing2Motion=Free
  twistLimitDegrees=45
  twistMotion=Free
 }
 {
  bone1Name=thigh_r
  bone2Name=pelvis
  swing1LimitDegrees=45
  swing1Motion=Free
  swing2LimitDegrees=45
  swing2Motion=Free
  twistLimitDegrees=45
  twistMotion=Free
 }
 {
  bone1Name=dyn_coat_bk_MSTR
  bone2Name=spine_02
  swing1LimitDegrees=45
  swing1Motion=Free
  swing2LimitDegrees=45
  swing2Motion=Free
  twistLimitDegrees=45
  twistMotion=Free
 }
 {
  bone1Name=dyn_coat_bk_l_MSTR
  bone2Name=spine_02
  swing1LimitDegrees=45
  swing1Motion=Free
  swing2LimitDegrees=45
  swing2Motion=Free
  twistLimitDegrees=45
  twistMotion=Free
 }
 {
  bone1Name=dyn_coat_l_MSTR
  bone2Name=spine_04
  swing1LimitDegrees=45
  swing1Motion=Free
  swing2LimitDegrees=45
  swing2Motion=Free
  twistLimitDegrees=45
  twistMotion=Free
 }
 {
  bone1Name=dyn_coat_r_MSTR
  bone2Name=spine_04
  swing1LimitDegrees=45
  swing1Motion=Free
  swing2LimitDegrees=45
  swing2Motion=Free
  twistLimitDegrees=45
  twistMotion=Free
 }
 {
  bone1Name=dyn_coat_bk_l_05
  bone2Name=dyn_coat_bk_05
  swing1LimitDegrees=45
  swing1Motion=Free
  swing2LimitDegrees=45
  swing2Motion=Free
  twistLimitDegrees=45
  twistMotion=Free
 }
 {
  bone1Name=dyn_coat_bk_r_05
  bone2Name=dyn_coat_bk_05
  swing1LimitDegrees=45
  swing1Motion=Free
  swing2LimitDegrees=45
  swing2Motion=Free
  twistLimitDegrees=45
  twistMotion=Free
 }
]
